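# Character master material for the LilSplit Sprinkles Super Level skin, layering the Season 27 Super Level effect onto the body and head.
# Chains shared PreFX, CharacterSkinFX and PostFX functions into MF_S27SL, masked by the smoothed SkinFX mask, then cinematic post FX.
# An IsHead static switch feeds white base color into SkinFX, then uses its contrasted, floored blue channel to blend the mouth back over gold.
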Material /BRCosmetics/Characters/Player/Male/Medium/Bodies/M_MED_LilSplit_Sprinkles/Skins/SuperLevel/Materials/M_LilSplit_SL
graph {
 properties {
  bAutomaticallySetUsageInEditor=false
  bUseMaterialAttributes=true
  bUsedWithMeshDeformer=true
  bUsedWithMorphTargets=true
  bUsedWithSkeletalMesh=true
  pixelDepthOffsetMode=PDOM_Legacy
  refractionMethod=RM_IndexOfRefraction
  shadingModel=MSM_FromMaterialExpression
  subsurfaceProfile=&SS_HeroSkin_02
  usedShadingModels=Default Lit | Subsurface | Clear Coat | Subsurface Profile | Cloth | Eye
 }
 nodes [
  n1 = MaterialFunctionCall() {materialFunction=&BaseFN_CharacterFunction_PreFX}
  n2 = Constant() {r=1}
  n3 = MaterialFunctionCall(
   PreskinnedPosition=@n1.Pre skinned Local Position
   Gmap=@n1.Pre skinned Local Position
   MaterialAttributes=StaticSwitchParameter(
    True=SetMaterialAttributes!(MaterialAttributes=@n1.Result, Base Color=@n2) {attributeSetTypes [#1]}
    False=@n1.Result
   ) {group=SuperLevel, parameterName=IsHead}
   AO=@n1.Pre skinned Local Position
  ) {materialFunction=&CharacterSkinFX}
  n4 = MaterialFunctionCall(AO=@n1.AO, Distance Blend=@n1.AO, Fresnel=@n1.AO, Gmap=@n1.AO, Pre Skinned Local Position=@n1.AO, Material=@n3.Result) {materialFunction=&BaseFN_CharacterFunction_PostFX}
  n5 = MaterialFunctionCall!(InMA=@n4.Result, FXMask=SmoothStep(Value=@n3.SkinFX_Mask) {constMax=0.7}, UseOriginalEmissive, PreSkinPos=@n1.Pre skinned Local Position) {materialFunction=/Game/Characters/Player/Common/MaterialFunctions/Superlevel_S27/MF_S27SL}
  n6 = ComponentMask(None=GetMaterialAttributes(None=@n3.Result) {attributeGetTypes [#1]}.Base Color) {b=true}
  n7 = Constant() {r=0.2}
  n8 = LinearInterpolate(
   A=GetMaterialAttributes(None=@n4.Result) {attributeGetTypes [#1]}.Base Color
   B=GetMaterialAttributes(None=@n5.OutMA) {attributeGetTypes [#1, #2]}.Base Color
   Alpha=Floor(None=MaterialFunctionCall(In=@n6, Contrast=@n7) {materialFunction=&CheapContrast}.Result)
  ) {desc=Blend Mouth back over gold}
  MaterialFunctionCall(In=StaticSwitchParameter(
    True=SetMaterialAttributes(MaterialAttributes=@n5.OutMA, Base Color=@n8) {attributeSetTypes [#1]}
    False=@n5.OutMA
   ) {group=SuperLevel, parameterName=IsHead}) {materialFunction=&CINE_MF_CharPostFX}
 ]
}
